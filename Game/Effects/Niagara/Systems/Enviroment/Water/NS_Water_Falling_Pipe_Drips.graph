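# Niagara effect of water falling from a pipe, with a ribbon stream, GPU drops and splashes below.
# System WaterForce and WaterVelocity shape the stream and drops, and User.WaterColor tints the water.
# A hard-coded 585-unit fall sets lifetimes, splash offsets and the height below which drops die.
# Splash_Drops also drives persistent audio at User.SplashLocation, toggled by User.PlayWaterSplash.
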NiagaraSystem /Game/Effects/Niagara/Systems/Enviroment/Water/NS_Water_Falling_Pipe_Drips
dependencies {
 usedDataInterfaces [&NiagaraDataInterfaceVector2DCurve, &NiagaraDataInterfaceCurve, &NiagaraDataInterfaceColorCurve]
 usedDynamicInputs [&Waveform, &Multiply_Vector]
 usedModules [
  &SystemState
  &self:SystemSpawnScript.SystemScriptSource.SystemScriptGraph.NiagaraNodeAssignment_0.SetVariables_#1
  &self:SystemSpawnScript.SystemScriptSource.SystemScriptGraph.NiagaraNodeAssignment_1.SetVariables_#2
 ]
}
emitters {
 FallingDrops {
  simTarget=GPUComputeSim
  emitterUpdateScript [
   &EmitterState(Enable Distance Culling=true, MaxDistance=3500, Module.bUseMaxDistance=true, Module.bUseMinDistance=false, Scalability Mode=::NewEnumerator1)
   &SpawnRate(
    Spawn Probability=0.75
    SpawnRate=&Multiply_Float(
     A=&RandomRangeFloat(Minimum=25, Maximum=40)
     B=&Lerp_Float(A=0.05, B=1, Alpha=$System.WaterForce)
    )
    Use Spawn Probability=true
   )
  ]
  particleSpawnScript [
   /SetVariables(Particles.Random=&UniformRangedFloat(Minimum=0, Maximum=1)) {isSetParameters=true, module=SetVariables_#3}
   &InitializeParticle(
    Color=$User.WaterColor
    Lifetime=&Multiply_Float(
     A=&RemapRange(
      Input Value To Remap=&MakeFloatFromVector(Vector=<0 0 -585>, Channel=::NewEnumerator2)
      Input Min=-1000
      Input Max=0
      Output Min=4
      Output Max=1
     )
     B=&RandomRangeFloat(Minimum=0.5, Maximum=1.1)
    )
    Sprite Size Max=<4 20>
    Sprite Size Min=<3 9>
    Sprite Size Mode=::NewEnumerator2
    UsePositionOffset=false
    Write Material Random=true
   ) {module=InitializeParticle001}
   &DynamicMaterialParameters(
    Index 0 Param 1=&UniformRangedFloat(Minimum=4, Maximum=10)
    Module.Param0WriteEnabled=true
    Module.Param0WriteEnabled3=true
    Module.Param1WriteEnabled=false
    Module.Param1WriteEnabled3=true
    Module.Param2WriteEnabled=false
    Module.Param2WriteEnabled3=true
    Module.Param3WriteEnabled=false
    Module.Param3WriteEnabled3=true
    Write Parameter Index 3=true
   )
   &ShapeLocation(MaskSpawnGroup=false, Override Seed=false, Sphere Radius=4, UseOwnerScale=true)
   &AddVelocity(Module.Use Falloff Distance=false, Override Seed=false, Velocity Mode=::NewEnumerator1, Velocity Speed=5)
   &AddVelocity(
    Cone Angle=&Multiply_Float(
     A=&Lerp_Float(A=0.5, B=4, Alpha=$System.WaterForce)
     B=&RemapRange(
      Input Value To Remap=&VectorLength(Vector=$System.WaterVelocity)
      Input Min=100
      Input Max=400
      Output Min=1
      Output Max=0
     )
    )
    Cone Axis=$System.WaterVelocity
    Module.UseConeFalloff=false
    Module.UseVelDistribution=false
    Override Seed=false
    Rotation Coordinate Space=::World
    Velocity Mode=::NewEnumerator2
    Velocity Speed=&Multiply_Float(
     A=&Multiply_Float(
      A=&RandomRangeFloat(Minimum=3, Maximum=3.3)
      B=&VectorLength(Vector=$System.WaterVelocity)
     )
     B=&Lerp_Float(A=0.5, B=1, Alpha=$System.WaterForce)
    )
   ) {module=AddVelocity001}
  ]
  particleUpdateScript [
   &ParticleState()
   &ScaleColor(
    Scale Alpha=&FloatFromCurve(
     FloatCurve=DataInterface() {propertyValues {Curve {keys [<at 0 0 RCIM_Cubic>, {interpMode=RCIM_Cubic, leaveTangent=0.002, time=0.2, value=1}, <at 1 0 RCIM_Cubic>]}}}
     Scale Curve=0.4
    )
    ScaleA=true
    ScaleRGB=true
   )
   &Drag(
    Drag=&FloatFromCurve(
     FloatCurve=DataInterface() {propertyValues {Curve {keys [<at 0 0.5 RCIM_Cubic>, {tangentMode=RCTM_User, time=0.8, value=1}]}}}
     Scale Curve=4
    )
    Use Linear Drag=true
    Use Rotational Drag=false
   )
   &GravityForce(Gravity=<0 0 -1200>)
   &SolveForcesAndVelocity(Clamp Velocity=false, Limit Acceleration=false)
   &KillParticles(Kill Particles=&SetBoolByFloatComparison(
     A=&MakeFloatFromVector(Vector=Linked() {linkedVariable {name=Particles.Position, type=NiagaraPosition}}, Channel=::NewEnumerator2)
     B=&Add_Float(
      A=-10
      B=&MakeFloatFromVector(
       Vector=&Add_Vector(A=<0 0 -585>, B=Linked() {linkedVariable {name=Engine.Owner.Position, type=NiagaraPosition}})
       Channel=::NewEnumerator2
      )
     )
     Comparison Type=::NewEnumerator6
    ))
  ]
  rendererClasses [&NiagaraSpriteRendererProperties]
  renderers [{rendererClass=&NiagaraSpriteRendererProperties}]
 }
 Splash_Drops {
  emitterSpawnScript [/Niagara/Modules/Audio/PlayPersistentAudio(
    Audio Playback Position:NiagaraPosition=$User.SplashLocation
    Audio Player:NiagaraDataInterfaceAudioPlayer=$User.Audio Player
    Audio Position Coordinate Space:ENiagaraCoordinateSpace=::Local
    Play Audio:NiagaraBool=$User.PlayWaterSplash
   )]
  emitterUpdateScript [
   &EmitterState(Enable Distance Culling=true, MaxDistance=3500, Module.bUseMaxDistance=true, Module.bUseMinDistance=false, Scalability Mode=::NewEnumerator1)
   &SpawnRate(SpawnRate=&RandomRangeFloat(Minimum=15, Maximum=30), Use Spawn Probability=false)
   &SpawnBurst_Instantaneous(
    Spawn Count=&RandomRangeInt(Minimum=0, Maximum=20)
    Spawn Time=0.05
    Use Spawn Probability=false
    UseLoopCountLimit=false
   )
   /Niagara/Modules/Audio/UpdatePersistentAudio(
    Audio Player:NiagaraDataInterfaceAudioPlayer=$User.Audio Player
    Position:NiagaraPosition=$User.SplashLocation
    Position Coordinate Space:ENiagaraCoordinateSpace=::Local
    Update Type:ENiagara_AudioParamType=::NewEnumerator5
   )
  ]
  particleSpawnScript [
   /SetVariables(Particles.Random=&UniformRangedFloat(Minimum=0, Maximum=1)) {isSetParameters=true, module=SetVariables_#3}
   &InitializeParticle(
    Color=$User.WaterColor
    Lifetime Max=1
    Lifetime Min=0.5
    Lifetime Mode=::NewEnumerator1
    Position Offset=<0 0 -585>
    Sprite Size=&Vector2DFromCurve(
     Vector2Curve=DataInterface() {propertyValues {
       XCurve {keys [<at -0.0011 5 RCIM_Cubic>,<at 0.9988 3 RCIM_Cubic>]}
       YCurve {keys [
         {arriveTangent=-30.8979, interpMode=RCIM_Cubic, leaveTangent=-30.8979, tangentMode=RCTM_User, time=0, value=15.3962}
         {arriveTangent=-6.8193, interpMode=RCIM_Cubic, leaveTangent=-6.8193, tangentMode=RCTM_User, time=0.2382, value=8}
         <at 0.9988 4 RCIM_Cubic>
        ]}
      }}
     CurveIndex=$Particles.Random
     ScaleCurve=<1 1>
    )
    Sprite Size Mode=::NewEnumerator4
    UsePositionOffset=true
    Write Material Random=true
   ) {module=InitializeParticle001}
   &DynamicMaterialParameters(
    Index 0 Param 1=&UniformRangedFloat(Minimum=4, Maximum=10)
    Module.Param0WriteEnabled=true
    Module.Param0WriteEnabled3=true
    Module.Param1WriteEnabled=false
    Module.Param1WriteEnabled3=true
    Module.Param2WriteEnabled=false
    Module.Param2WriteEnabled3=true
    Module.Param3WriteEnabled=false
    Module.Param3WriteEnabled3=true
    Write Parameter Index 3=true
   )
   &ShapeLocation(
    Cylinder Height=1
    Cylinder Radius=10
    MaskSpawnGroup=false
    Override Seed=false
    Shape Primitive=::NewEnumerator1
    Surface Only=false
    UseOwnerScale=true
   )
   &AddVelocity(
    Cone Angle=25
    Cone Axis=<0 0 1>
    Module.UseConeFalloff=false
    Module.UseVelDistribution=false
    Override Seed=false
    Velocity Mode=::NewEnumerator2
    Velocity Speed=&RandomRangeFloat(Minimum=80, Maximum=450)
   ) {module=AddVelocity001}
  ]
  particleUpdateScript [
   &ParticleState()
   &ScaleColor(
    Scale Alpha=&FloatFromCurve(
     FloatCurve=DataInterface() {propertyValues {Curve {keys [
         <at 0 0 RCIM_Cubic>
         {arriveTangent=0.002, interpMode=RCIM_Cubic, leaveTangent=0.002, time=0.1183, value=1.0113}
         <at 1.0011 0.002 RCIM_Cubic>
        ]}}}
     Scale Curve=0.6
    )
    ScaleA=true
    ScaleRGB=true
   )
   &PointForce(
    Force Origin=&Add_Vector(A=Linked() {linkedVariable {name=Engine.Owner.Position, type=NiagaraPosition}}, B=<0 0 -585>)
    Force Strength=200
    Module.Use Falloff=false
    Module.UseLinearForce=false
   )
   &Drag(Drag=&RandomRangeFloat(Minimum=1, Maximum=1.6), Use Linear Drag=true, Use Rotational Drag=false)
   &GravityForce()
   &KillParticles(Kill Particles=&SetBoolByFloatComparison(
     A=&MakeFloatFromVector(Vector=Linked() {linkedVariable {name=Particles.Position, type=NiagaraPosition}}, Channel=::NewEnumerator2)
     B=&Add_Float(
      A=-10
      B=&MakeFloatFromVector(
       Vector=&Add_Vector(A=<0 0 -585>, B=Linked() {linkedVariable {name=Engine.Owner.Position, type=NiagaraPosition}})
       Channel=::NewEnumerator2
      )
     )
     Comparison Type=::NewEnumerator6
    ))
   &SolveForcesAndVelocity(Clamp Velocity=false, Limit Acceleration=false)
  ]
  rendererClasses [&NiagaraSpriteRendererProperties]
  renderers [{rendererClass=&NiagaraSpriteRendererProperties}]
 }
 Splash_SubUV {
  emitterUpdateScript [
   &EmitterState()
   &SpawnRate(SpawnRate=&RandomRangeFloat(Minimum=4, Maximum=10), Use Spawn Probability=false)
  ]
  particleSpawnScript [
   &InitializeParticle(
    Color=$User.WaterColor
    Lifetime Max=0.55
    Lifetime Min=0.3
    Lifetime Mode=::NewEnumerator1
    Position Offset=<0 0 -585>
    Sprite Size Max=<120 100>
    Sprite Size Min=<70 60>
    Sprite Size Mode=::NewEnumerator2
    UsePositionOffset=true
    Write Material Random=true
   ) {module=InitializeParticle001}
   &AddVelocityInCone(
    Cone Angle=50
    Cone Axis=<0 0 1>
    Module.Use Velocity Falloff On Cone Axis=false
    Velocity Distribution Along Cone Axis=0.25
    Velocity Strength=&UniformRangedFloat(Minimum=100, Maximum=200)
   )
   &SubUVAnimation(End Frame=3, SubUV Animation Mode:ENiagara_SubUVLookupMode=::NewEnumerator2)
  ]
  particleUpdateScript [
   &ParticleState()
   &ScaleSpriteSize(Scale Factor=&Vector2DFromCurve(
     Vector2Curve=DataInterface() {propertyValues {
       XCurve {keys [
         {arriveTangent=0.4604, interpMode=RCIM_Cubic, leaveTangent=0.4604, tangentMode=RCTM_User, time=0, value=0.05}
         {arriveTangent=1.2987, interpMode=RCIM_Cubic, leaveTangent=1.2987, tangentMode=RCTM_User, time=1, value=1}
        ]}
       YCurve {keys [
         {arriveTangent=2.0407, interpMode=RCIM_Cubic, leaveTangent=2.0407, tangentMode=RCTM_User, time=0, value=0.05}
         {arriveTangentWeight=0.3333, interpMode=RCIM_Cubic, leaveTangentWeight=0.0833, tangentMode=RCTM_User, time=0.75, value=1}
         {arriveTangent=-2.358, interpMode=RCIM_Cubic, leaveTangent=-2.358, tangentMode=RCTM_User, time=1, value=0.7}
        ]}
      }}
     ScaleCurve=<1 1>
    ))
   &DynamicMaterialParameters(
    Index 0 Param 1=&FloatFromCurve(
     FloatCurve=DataInterface() {propertyValues {Curve {keys [
         {arriveTangent=-0.5095, interpMode=RCIM_Cubic, leaveTangent=-0.5095, tangentMode=RCTM_User, time=0, value=1}
         {arriveTangent=-1.3585, interpMode=RCIM_Cubic, leaveTangent=-1.3585, tangentMode=RCTM_User, time=1, value=0}
        ]}}}
     Scale Curve=1.5
    )
    Module.Param0WriteEnabled=true
    Module.Param1WriteEnabled=false
    Module.Param2WriteEnabled=false
    Module.Param3WriteEnabled=false
   )
   &Color(Color=&MakeLinearColorFromVectorAndFloat(
     "Vector (RGB)"=<1 1 1>
     "Float (Alpha)"=&FloatFromCurve(FloatCurve=DataInterface() {propertyValues {Curve {keys [<at 0.4 1.0113 RCIM_Cubic>,<at 1.0011 0.002>]}}})
    ))
   &Drag(Drag=3, Use Linear Drag=true, Use Rotational Drag=false)
   &WindForce(Air Resistance=0.1, Wind Speed=<0 0 -1>, Wind Speed Scale=100)
   &SpriteRotationRate()
   &SolveForcesAndVelocity(Clamp Velocity=false, Limit Acceleration=false)
  ]
  rendererClasses [&NiagaraSpriteRendererProperties]
  renderers [{rendererClass=&NiagaraSpriteRendererProperties}]
 }
 WaterStream {
  emitterUpdateScript [
   &EmitterState(
    Life Cycle Mode=::NewEnumerator1
    Loop Duration=&UniformRangedFloat(Minimum=4, Maximum=6)
    Recalculate Duration Each Loop=true
    UseLoopDelay=false
   )
   &SpawnRate(SpawnRate=25, Use Spawn Probability=false)
  ]
  particleSpawnScript [
   /Niagara/Modules/Spawn/Initialization/InitializeParticle(
    Lifetime=0.75
    Module.Write Color=true
    Module.Write Lifetime=true
    Module.Write Mass=true
    Module.Write Material Random=true
    Module.Write Position=false
    Module.Write Scale=false
    Module.Write SpriteRotation=false
    Module.Write SpriteSize=true
    Position=&SimulationPosition()
   ) {enabled=false}
   &InitializeRibbon(
    Color=&MultiplyLinearColor(
     Linear Color=$User.WaterColor
     Scale Factor=&MakeFloat4FromVectorAndFloat("Vector (RGB)"=<1 1 1>, "Float (Alpha)"=$System.WaterForce)
    )
    Lifetime=&RemapRange(
     Input Value To Remap=&MakeFloatFromVector(Vector=<0 0 -585>, Channel=::NewEnumerator2)
     Input Min=-1000
     Input Max=0
     Output Min=3.2
     Output Max=0.4
    )
    Ribbon Width=&Lerp_Float(A=2, B=35, Alpha=$System.WaterForce)
    UsePositionOffset=false
    Write Material Random=true
   )
   &AddVelocity(
    Rotation Coordinate Space=::World
    Velocity=$System.WaterVelocity
    Velocity Speed Scale=&Lerp_Float(A=2, B=2.3, Alpha=$System.WaterForce)
   )
  ]
  particleUpdateScript [
   &ParticleState()
   &GravityForce(Gravity=<0 0 -1200>)
   &Drag(
    Drag=&FloatFromCurve(
     FloatCurve=DataInterface() {propertyValues {Curve {keys [<at 0 0.5 RCIM_Cubic>, {tangentMode=RCTM_User, time=0.8, value=1}]}}}
     Scale Curve=4
    )
    Use Linear Drag=true
    Use Rotational Drag=false
   )
   &SolveForcesAndVelocity(Clamp Velocity=false, Limit Acceleration=false)
   &DynamicMaterialParameters(
    Index 0 Param 1=0.5
    Index 0 Param 2=10
    Module.Param0WriteEnabled=true
    Module.Param1WriteEnabled=true
    Module.Param2WriteEnabled=true
    Module.Param3WriteEnabled=true
   )
   &ScaleColor(
    Scale Alpha=&FloatFromCurve(
     FloatCurve=DataInterface() {propertyValues {Curve {keys [
         {arriveTangent=-1.4341, interpMode=RCIM_Cubic, time=0, value=0}
         {arriveTangent=-1.0607, interpMode=RCIM_Cubic, leaveTangent=-1.0607, tangentMode=RCTM_User, time=0.1, value=0.8}
         {interpMode=RCIM_Cubic, tangentMode=RCTM_Break, time=1, value=0.05}
        ]}}}
     Scale Curve=0.8
    )
    Scale RGB=&MakeVectorFromLinearColorRGB(Color=&ColorFromCurve(ColorCurve=DataInterface() {propertyValues {
        AlphaCurve {keys []}
        BlueCurve {keys [{arriveTangent=-1.0847, interpMode=RCIM_Cubic, time=0, value=1}, <at 0.2431 0.3281 RCIM_Cubic>, <at 0.5613 1 RCIM_Cubic>]}
        GreenCurve {keys [
          {arriveTangent=-3.7763, interpMode=RCIM_Cubic, time=0, value=1}
          {arriveTangent=-0.2768, interpMode=RCIM_Cubic, leaveTangent=-0.2768, time=0.2431, value=0.3149}
          <at 0.5613 0.8446 RCIM_Cubic>
         ]}
        RedCurve {keys [
          {arriveTangent=-8.1645, interpMode=RCIM_Cubic, time=0, value=0.9792}
          {arriveTangent=-0.5523, interpMode=RCIM_Cubic, leaveTangent=-0.5523, time=0.2431, value=0.2751}
          <at 0.5613 0.6691 RCIM_Cubic>
         ]}
       }}))
    ScaleA=true
    ScaleRGB=true
   )
   &ScaleRibbonWidth(Ribbon Width Scale=&FloatFromCurve(FloatCurve=DataInterface() {propertyValues {Curve {keys [
         {arriveTangent=13.5248, interpMode=RCIM_Cubic, leaveTangent=13.5248, tangentMode=RCTM_User, time=0, value=0}
         {arriveTangent=0.3927, interpMode=RCIM_Cubic, leaveTangent=0.3927, tangentMode=RCTM_User, time=0.07, value=0.7}
         {arriveTangent=0.1646, interpMode=RCIM_Cubic, leaveTangent=0.1646, tangentMode=RCTM_User, time=1, value=1}
        ]}}}))
  ]
  rendererClasses [&NiagaraSpriteRendererProperties, &NiagaraRibbonRendererProperties]
  renderers [
   {rendererClass=&NiagaraSpriteRendererProperties}
   {rendererClass=&NiagaraRibbonRendererProperties, rendererIndex=1}
  ]
 }
}
summary {
 systemName=NS_Water_Falling_Pipe_Drips
 emitters [
  {emitterName=Splash_Drops, rendererClasses [&NiagaraSpriteRendererProperties]}
  {emitterName=WaterStream, rendererClasses [&NiagaraSpriteRendererProperties, &NiagaraRibbonRendererProperties]}
  {emitterName=FallingDrops, simTarget=GPUComputeSim, rendererClasses [&NiagaraSpriteRendererProperties]}
  {emitterName=Splash_SubUV, rendererClasses [&NiagaraSpriteRendererProperties]}
 ]
 userVariables [
  User.Audio Player {
   type=NiagaraDataInterfaceAudioPlayer
   description=""
   defaultValue {
    struct=NiagaraExt_VariableValue_DataInterface
    value {
     dataInterface=&self:NiagaraDataInterfaceAudioPlayer_0
     dataInterfaceClass=&NiagaraDataInterfaceAudioPlayer
    }
   }
  }
  User.PlayWaterSplash {type=NiagaraBool, defaultValue=true}
  User.SplashLocation {type=NiagaraPosition, description="", defaultValue {}}
  User.WaterColor {type=LinearColor, description="", defaultValue {struct=LinearColor, value=<rgba 0.6 0.4745 0.3299 1>}}
 ]
}
